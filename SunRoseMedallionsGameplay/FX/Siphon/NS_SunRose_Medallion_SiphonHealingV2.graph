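# Niagara effect for the SunRose medallion's siphon healing, showing a character gaining health or shield.
# User.IsShieldIncreasing lerps most colors from green to blue and, via spawn probability, swaps cross sparkles for squares, both spawned in a ring rising with scaled system age.
# An edge-glow mesh follows one bone and wash sprites track 17 partitioned bones, alongside wisp meshes, layered flares and a brief light.
# Shared System variables drive color, opacity and scale.
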NiagaraSystem /SunRoseMedallionsGameplay/FX/Siphon/NS_SunRose_Medallion_SiphonHealingV2
dependencies {
 usedDataInterfaces [&NiagaraDataInterfaceCurve, &NiagaraDataInterfaceSkeletalMesh, &NiagaraDataInterfaceArrayFloat3]
 usedDynamicInputs [
  &FNTimeOfDayScalar
  &RemapRange
  &ClampFloat
 ]
 usedModules [
  &self:SystemSpawnScript.SystemScriptSource.SystemScriptGraph.NiagaraNodeAssignment_0.SetVariables_#1
  &self:SystemSpawnScript.SystemScriptSource.SystemScriptGraph.NiagaraNodeAssignment_1.SetVariables_#2
  &self:SystemSpawnScript.SystemScriptSource.SystemScriptGraph.NiagaraNodeAssignment_2.SetVariables_#3
  &self:SystemSpawnScript.SystemScriptSource.SystemScriptGraph.NiagaraNodeAssignment_3.SetVariables_#4
  &self:SystemSpawnScript.SystemScriptSource.SystemScriptGraph.NiagaraNodeAssignment_9.SetVariables_#5
  &SystemState
  &self:SystemSpawnScript.SystemScriptSource.SystemScriptGraph.NiagaraNodeAssignment_4.SetVariables_#6
  &self:SystemSpawnScript.SystemScriptSource.SystemScriptGraph.NiagaraNodeAssignment_5.SetVariables_#7
  &self:SystemSpawnScript.SystemScriptSource.SystemScriptGraph.NiagaraNodeAssignment_6.SetVariables_#8
  &self:SystemSpawnScript.SystemScriptSource.SystemScriptGraph.NiagaraNodeAssignment_7.SetVariables_#9
  &self:SystemSpawnScript.SystemScriptSource.SystemScriptGraph.NiagaraNodeAssignment_8.SetVariables_#10
 ]
}
emitters {
 CrossSparkles {
  emitterUpdateScript [
   &EmitterState()
   &SpawnRate(
    Spawn Probability=&Lerp_Float(A=1, B=0, Alpha=$User.IsShieldIncreasing)
    SpawnRate=&FloatFromCurve(
     FloatCurve=DataInterface() {propertyValues {Curve {keys [
         {interpMode=RCIM_Cubic, tangentMode=RCTM_User, time=0, value=0}
         {interpMode=RCIM_Cubic, tangentMode=RCTM_User, time=0.05, value=1}
         {interpMode=RCIM_Cubic, tangentMode=RCTM_User, time=0.5, value=1}
         <at 0.6 0 RCIM_Cubic>
        ]}}}
     CurveIndex=$System.NormalizedAgeScaled
     Scale Curve=30
    )
    Use Spawn Probability=true
   )
   &SpawnRate(
    Spawn Probability=&Lerp_Float(A=1, B=0, Alpha=$User.IsShieldIncreasing)
    SpawnGroup=1
    SpawnRate=&FloatFromCurve(
     FloatCurve=DataInterface() {propertyValues {Curve {keys [
         {interpMode=RCIM_Cubic, tangentMode=RCTM_User, time=0, value=0}
         {interpMode=RCIM_Cubic, tangentMode=RCTM_User, time=0.05, value=1}
         {interpMode=RCIM_Cubic, tangentMode=RCTM_User, time=0.5, value=1}
         <at 0.6 0 RCIM_Cubic>
        ]}}}
     CurveIndex=$System.NormalizedAgeScaled
     Scale Curve=15
    )
    Use Spawn Probability=true
   ) {module=SpawnRate001}
  ]
  particleSpawnScript [
   /SetVariables(Particles.VisibilityTag=&SpawnGroup()) {isSetParameters=true}
   &InitializeParticle(
    Color=$System.ColorA
    Lifetime=&MaskFloatBySpawnGroup(
     Masked Float=&RandomRangeFloat(Minimum=0.15, Maximum=0.4)
     Passthrough Float=&RandomRangeFloat(Minimum=0.5, Maximum=0.6)
    )
    Position Offset=&Add_Vector(
     A=<0 0 10>
     B=&Multiply_VectorByFloat(
      Vector=<0 0 200>
      Float=&FloatFromCurve(
       FloatCurve=DataInterface() {propertyValues {Curve {keys [{tangentMode=RCTM_User, time=0, value=0}, {tangentMode=RCTM_User, time=0.8, value=1}, <at 1 0>]}}}
       CurveIndex=$System.NormalizedAgeScaled
      )
     )
    )
    Sprite Size Mode=::NewEnumerator1
    Uniform Sprite Size Max=3
    Uniform Sprite Size Min=2
    UsePositionOffset=true
    Write Material Random=true
   )
   &CylinderLocation(
    Cylinder Height=10
    Cylinder Radius=&Multiply_Float(A=&RandomRangeFloat(Minimum=15, Maximum=30), B=$System.GlowPanScale)
    MaskSpawnGroup=false
    Override Seed=false
    UseOwnerScale=true
   )
   &AddVelocity(
    Velocity=&MaskVectorBySpawnGroup(Masked Vector=<0 0 0>, Passthrough Vector=<0 0 1>)
    Velocity Speed Scale=&RandomRangeFloat(Minimum=40, Maximum=60)
   )
  ]
  particleUpdateScript [
   &Drag(Drag=&RandomRangeFloat(Minimum=4, Maximum=6), Use Linear Drag=true, Use Rotational Drag=false)
   &SolveForcesAndVelocity(Clamp Velocity=false, Limit Acceleration=false)
   &ParticleState()
   &ScaleColor(
    Scale Alpha=&FloatFromCurve(
     FloatCurve=DataInterface() {propertyValues {Curve {keys [
         {arriveTangent=4.9067, interpMode=RCIM_Cubic, leaveTangent=4.9067, tangentMode=RCTM_User, time=0, value=0}
         {arriveTangentWeight=0.181, interpMode=RCIM_Cubic, leaveTangentWeight=0.7242, tangentMode=RCTM_User, time=0.15, value=1}
         {arriveTangent=-2.0732, interpMode=RCIM_Cubic, leaveTangent=-2.0732, tangentMode=RCTM_User, time=0.4, value=0.4}
         {interpMode=RCIM_Cubic, leaveTangent=-0.0405, time=1, value=0}
        ]}}}
     Scale Curve=$System.OpacityScalar
    )
    ScaleA=true
    ScaleRGB=true
   )
   &ScaleSpriteSize(
    Uniform Curve Scale=3
    Uniform Curve Sprite Scale=DataInterface() {propertyValues {Curve {keys [
        {arriveTangent=4.1785, interpMode=RCIM_Cubic, leaveTangent=4.1785, tangentMode=RCTM_User, time=0, value=0}
        {arriveTangent=1.1179, interpMode=RCIM_Cubic, leaveTangent=1.1179, tangentMode=RCTM_User, time=0.2042, value=0.6228}
        {arriveTangent=0.0224, interpMode=RCIM_Cubic, leaveTangent=0.0224, tangentMode=RCTM_User, time=1, value=1}
       ]}}}
   )
   /SetVariables(Particles.SpriteSize_Sparkles:Vector2f=&Multiply_Vector2D_ByFloat(
     Float=&FloatFromCurve(
      FloatCurve=DataInterface() {propertyValues {Curve {keys [
          {arriveTangent=9.0725, interpMode=RCIM_Cubic, time=0, value=0}
          <at 0.25 0.9847 RCIM_Cubic>
          {arriveTangent=-0.0405, interpMode=RCIM_Cubic, leaveTangent=-0.0405, tangentMode=RCTM_User, time=1, value=0}
         ]}}}
      Scale Curve=3
     )
     Vector2D=$Particles.Initial.SpriteSize
    )) {isSetParameters=true}
   /SetVariables(Particles.Color_Sparkles:LinearColor=&MultiplyLinearColor(
     Linear Color=$Particles.Initial.Color
     Scale Factor=&MakeFloat4FromVectorAndFloat(
      "Vector (RGB)"=<2 2 2>
      "Float (Alpha)"=&FloatFromCurve(FloatCurve=DataInterface() {propertyValues {Curve {keys [
           {arriveTangent=4.9067, interpMode=RCIM_Cubic, time=0, value=0}
           {arriveTangentWeight=0.181, interpMode=RCIM_Cubic, leaveTangentWeight=0.7242, tangentMode=RCTM_User, time=0.25, value=1}
           {interpMode=RCIM_Cubic, leaveTangent=-0.0405, time=1, value=0}
          ]}}})
     )
    )) {isSetParameters=true}
   &CameraOffset(Camera Offset Amount=25)
  ]
  rendererClasses [&NiagaraSpriteRendererProperties]
  renderers [
   {rendererClass=&NiagaraSpriteRendererProperties}
   {rendererClass=&NiagaraSpriteRendererProperties, rendererIndex=1}
  ]
 }
 EdgeGlow {
  emitterSpawnScript [/SetVariables(Emitter.OutlineGlowColor:LinearColor=&LerpLinearColors(StartColor=<rgba 0.1193 2 0.4063 1>, EndColor=<rgba 0 0.927 2 1>, LerpFactor=$User.IsShieldIncreasing)) {isSetParameters=true}]
  emitterUpdateScript [
   &EmitterState()
   &SpawnBurst_Instantaneous(Spawn Time=0.1, Use Spawn Probability=false, UseLoopCountLimit=false) {module=SpawnBurst_Instantaneous001}
  ]
  particleSpawnScript [
   &InitializeParticle(
    Lifetime=1.5
    Mesh Scale=<0.5 0.5 0.85>
    Mesh Scale Mode=::NewEnumerator4
    Sprite Size Mode=::NewEnumerator3
    Uniform Sprite Size=500
    UsePositionOffset=false
    Write Material Random=true
   )
   &InitialMeshOrientation(Mesh Axis To Orient=<0 0 1>, Mesh Orientation Mode=::NewEnumerator2, Use Rotation Vector=true)
  ]
  particleUpdateScript [
   &ParticleState()
   &SkeletalMeshLocation(
    Bone / Socket Index=16
    Bone Sampling Mode=::NewEnumerator4
    MaskSpawnGroup=false
    Orientation Sampling=::NewEnumerator1
    Sampled Position Offset=<0 0 -60>
    Sampled Position Offset Coordinate Space:ENiagaraSkelMeshTransforms=::NewEnumerator1
    Skeletal Mesh=$System.Skeletal Mesh
   )
   &CameraOffset(Camera Offset Amount=38.6542)
   &ScaleColor(
    Scale Alpha=&FloatFromCurve(
     FloatCurve=DataInterface() {propertyValues {Curve {keys [
         {interpMode=RCIM_Cubic, tangentMode=RCTM_User, time=0, value=0}
         {arriveTangent=-4.5912, interpMode=RCIM_Cubic, leaveTangent=-4.5912, tangentMode=RCTM_User, time=0.1, value=1.1}
         {tangentMode=RCTM_User, time=1, value=0}
        ]}}}
     Scale Curve=$System.OpacityScalar
    )
    Scale RGB=&VectorFromFloat(Value=&FloatFromCurve(FloatCurve=DataInterface() {propertyValues {Curve {keys [
          {arriveTangent=9.0725, interpMode=RCIM_Cubic, leaveTangent=9.0725, tangentMode=RCTM_User, time=0, value=0}
          {arriveTangent=-2.5249, interpMode=RCIM_Cubic, leaveTangent=-2.5249, tangentMode=RCTM_User, time=0.1408, value=0.9847}
          {arriveTangent=-0.0405, interpMode=RCIM_Cubic, leaveTangent=-0.0405, tangentMode=RCTM_User, time=1, value=0}
         ]}}}))
    ScaleA=true
    ScaleRGB=true
   )
   &ScaleColor(
    Scale RGB=&VectorFromFloat(Value=&FloatFromCurve(
      FloatCurve=DataInterface() {propertyValues {Curve {keys [
          {interpMode=RCIM_Cubic, tangentMode=RCTM_Break, time=0, value=1}
          {arriveTangent=1.4855, interpMode=RCIM_Cubic, tangentMode=RCTM_Break, time=3, value=3}
         ]}}}
      CurveIndex=$Emitter.Age
     ))
    ScaleA=true
    ScaleRGB=true
   ) {enabled=false, module=ScaleColor003}
   &DynamicMaterialParameters(
    Index 0 Param 1=&FloatFromCurve(FloatCurve=DataInterface() {propertyValues {Curve {keys [{interpMode=RCIM_Cubic, tangentMode=RCTM_User, time=0, value=0}, {interpMode=RCIM_Cubic, tangentMode=RCTM_User, time=1, value=1}]}}})
    Module.Index0Param0Write=true
    Module.Index0Param1Write=true
    Module.Index0Param2Write=true
    Module.Index0Param3Write=true
   )
  ]
  rendererClasses [&NiagaraMeshRendererProperties]
  renderers [
   {rendererClass=&NiagaraMeshRendererProperties, rendererIndex=0}
   {rendererClass=&NiagaraMeshRendererProperties, rendererIndex=1}
  ]
 }
 Flares {
  emitterUpdateScript [
   &EmitterState()
   &SpawnBurst_Instantaneous(Spawn Time=0.15, Use Spawn Probability=false, UseLoopCountLimit=false)
  ]
  particleSpawnScript [
   &InitializeParticle(
    Color=&LerpLinearColors(StartColor=<rgba 0 3.434 1.1275 2>, EndColor=<rgba 0.0156 1.0104 3 1>, LerpFactor=$User.IsShieldIncreasing)
    Lifetime=0.2
    Sprite Rotation Angle=0.2
    Sprite Rotation Mode=::NewEnumerator1
    Sprite Size Mode=::NewEnumerator3
    Uniform Sprite Size=150
    UsePositionOffset=false
    Write Material Random=true
   )
   &CameraOffset(Camera Offset Amount=15)
   /SetVariables(Particles.AnamorphicSize:Vector2f=<240 7>) {isSetParameters=true}
   &SystemLocation(MaskSpawnGroup=false, Offset=<0 0 120>)
  ]
  particleUpdateScript [
   &ParticleState()
   &ScaleColor(
    Scale Alpha=&FloatFromCurve(
     FloatCurve=DataInterface() {propertyValues {Curve {keys [
         {arriveTangent=0.6364, interpMode=RCIM_Cubic, time=0, value=0}
         {arriveTangent=1.4286, interpMode=RCIM_Cubic, leaveTangent=1.4286, time=0.5, value=0.25}
         {arriveTangent=-8.2069, interpMode=RCIM_Cubic, leaveTangent=-8.2069, tangentMode=RCTM_User, time=0.7, value=1}
         {arriveTangent=-0.0405, interpMode=RCIM_Cubic, leaveTangent=-0.0405, tangentMode=RCTM_User, time=1, value=0}
        ]}}}
     Scale Curve=$System.OpacityScalar
    )
    ScaleA=true
    ScaleRGB=true
   )
   &ScaleSpriteSize(Uniform Curve Sprite Scale=DataInterface() {propertyValues {Curve {keys [
        {interpMode=RCIM_Cubic, leaveTangentWeight=0.1863, tangentMode=RCTM_User, time=0, value=0.4}
        {
         arriveTangent=0.3825
         arriveTangentWeight=0.2415
         interpMode=RCIM_Cubic
         leaveTangent=0.3825
         leaveTangentWeight=0.0659
         tangentMode=RCTM_User
         time=0.55
         value=0.5
        }
        {
         arriveTangent=-1.9384
         arriveTangentWeight=0.0547
         interpMode=RCIM_Cubic
         leaveTangent=-1.9384
         leaveTangentWeight=0.1094
         tangentMode=RCTM_User
         time=0.7
         value=1
        }
        {arriveTangentWeight=0.1414, interpMode=RCIM_Cubic, tangentMode=RCTM_User, time=1, value=0.7}
       ]}}})
   /SetVariables(
    Particles.Anamorphic Color:LinearColor=&MakeLinearColorFromVectorAndFloat(
     "Vector (RGB)"=&MakeVectorFromLinearColorRGB(Color=&LerpLinearColors(StartColor=<rgba 0 1.2 0.394 1>, EndColor=<rgba 0.0156 1.0104 3 1>, LerpFactor=$User.IsShieldIncreasing))
     "Float (Alpha)"=&Multiply_Float(
      A=0.1
      B=&FloatFromCurve(
       FloatCurve=DataInterface() {propertyValues {Curve {keys [
           {arriveTangent=0.6364, interpMode=RCIM_Cubic, leaveTangent=0.6364, time=0, value=0}
           {arriveTangent=1.4286, interpMode=RCIM_Cubic, leaveTangent=1.4286, time=0.55, value=0.35}
           {arriveTangent=-8.2069, interpMode=RCIM_Cubic, leaveTangent=-8.2069, tangentMode=RCTM_User, time=0.7, value=1}
           {arriveTangent=-0.0405, interpMode=RCIM_Cubic, leaveTangent=-0.0405, tangentMode=RCTM_User, time=1, value=0}
          ]}}}
       Scale Curve=$System.OpacityScalar
      )
     )
    )
    Particles.CoronaColor:LinearColor=&MakeLinearColorFromVectorAndFloat(
     "Vector (RGB)"=&MakeVectorFromLinearColorRGB(Color=&LerpLinearColors(StartColor=<rgba 0 1.2 0.394 1>, EndColor=<rgba 0.0156 1.0104 3 1>, LerpFactor=$User.IsShieldIncreasing))
     "Float (Alpha)"=&FloatFromCurve(
      FloatCurve=DataInterface() {propertyValues {Curve {keys [
          {interpMode=RCIM_Cubic, leaveTangentWeight=0.3578, tangentMode=RCTM_User, time=0, value=0}
          {arriveTangent=0.6909, interpMode=RCIM_Cubic, leaveTangent=0.6909, tangentMode=RCTM_User, time=0.5, value=0.95}
          {arriveTangentWeight=0.1431, interpMode=RCIM_Cubic, leaveTangentWeight=0.2147, tangentMode=RCTM_User, time=0.7, value=1}
          {arriveTangentWeight=0.348, interpMode=RCIM_Cubic, tangentMode=RCTM_User, time=1, value=0}
         ]}}}
      Scale Curve=$System.OpacityScalar
     )
    )
    Particles.CoronaDynParameters:Vector4f=&MakeVector4(
     X=&FloatFromCurve(FloatCurve=DataInterface() {propertyValues {Curve {keys [
          {interpMode=RCIM_Cubic, leaveTangentWeight=0.1904, tangentMode=RCTM_User, time=0, value=1.4}
          {
           arriveTangent=-1.2888
           arriveTangentWeight=0.2753
           interpMode=RCIM_Cubic
           leaveTangent=-1.2888
           leaveTangentWeight=0.0918
           tangentMode=RCTM_User
           time=0.525
           value=1.175
          }
          {arriveTangent=-1.9474, interpMode=RCIM_Cubic, leaveTangent=-1.9474, time=0.7, value=0.55}
          {arriveTangentWeight=0.1414, interpMode=RCIM_Cubic, tangentMode=RCTM_User, time=1, value=0.25}
         ]}}})
     Y=8
     Z=0.6
     W=1
    )
    Particles.CoronaSize:Vector2f=&Vector2DFromFloat(Value=&Multiply_Float(
      A=75
      B=&FloatFromCurve(FloatCurve=DataInterface() {propertyValues {Curve {keys [
           {interpMode=RCIM_Cubic, leaveTangentWeight=0.3578, tangentMode=RCTM_User, time=0, value=0}
           {
            arriveTangent=1.1981
            arriveTangentWeight=0.2906
            interpMode=RCIM_Cubic
            leaveTangent=1.1981
            leaveTangentWeight=0.1163
            tangentMode=RCTM_User
            time=0.5
            value=0.95
           }
           {arriveTangentWeight=0.1431, interpMode=RCIM_Cubic, leaveTangentWeight=0.2147, tangentMode=RCTM_User, time=0.7, value=1}
          ]}}})
     ))
    Particles.DistanceGlowSize:Vector2f=&Vector2DFromFloat(Value=&Multiply_Float(
      A=350
      B=&FloatFromCurve(FloatCurve=DataInterface() {propertyValues {Curve {keys [
           {interpMode=RCIM_Cubic, leaveTangentWeight=0.348, tangentMode=RCTM_User, time=0, value=0.7}
           {arriveTangent=0.7107, arriveTangentWeight=0.348, interpMode=RCIM_Cubic, leaveTangent=0.7107, tangentMode=RCTM_User, time=1, value=1}
          ]}}})
     ))
    Particles.DistantGlowColor:LinearColor=&MakeLinearColorFromVectorAndFloat(
     "Vector (RGB)"=&MakeVectorFromLinearColorRGB(Color=&LerpLinearColors(StartColor=<rgba 0 1.2 0.394 1>, EndColor=<rgba 0.0156 1.0104 3 1>, LerpFactor=$User.IsShieldIncreasing))
     "Float (Alpha)"=&FloatFromCurve(
      FloatCurve=DataInterface() {propertyValues {Curve {keys [
          {arriveTangent=9.0725, interpMode=RCIM_Cubic, leaveTangent=9.0725, tangentMode=RCTM_User, time=0, value=0}
          {arriveTangentWeight=0.0905, interpMode=RCIM_Cubic, leaveTangentWeight=0.8147, tangentMode=RCTM_User, time=0.1, value=0.3}
          {arriveTangent=-0.0405, interpMode=RCIM_Cubic, leaveTangent=-0.0405, tangentMode=RCTM_User, time=1, value=0}
         ]}}}
      Scale Curve=$System.OpacityScalar
     )
    )
   ) {isSetParameters=true}
  ]
  rendererClasses [&NiagaraSpriteRendererProperties]
  renderers [
   {rendererClass=&NiagaraSpriteRendererProperties}
   {rendererClass=&NiagaraSpriteRendererProperties, rendererIndex=1}
   {rendererClass=&NiagaraSpriteRendererProperties, rendererIndex=2}
   {rendererClass=&NiagaraSpriteRendererProperties, rendererIndex=3}
  ]
 }
 Light {
  emitterUpdateScript [
   &EmitterState()
   &SpawnBurst_Instantaneous(Spawn Time=0.1, Use Spawn Probability=false, UseLoopCountLimit=false)
  ]
  particleSpawnScript [
   &InitializeParticle(
    Color=&LerpLinearColors(StartColor=<rgba 0.0331 1 0.1644 1>, EndColor=<rgba 0.0458 0.409 1 1>, LerpFactor=$User.IsShieldIncreasing)
    Lifetime=0.2
    Position Offset=<0 0 -50>
    UsePositionOffset=true
    Write Material Random=true
   )
   &SystemLocation(MaskSpawnGroup=false, Offset=<0 0 20>)
  ]
  particleUpdateScript [
   &ParticleState()
   /SetVariables(Particles.LightExponent=2, Particles.LightRadius=150) {isSetParameters=true}
   &ScaleColor(
    Scale Alpha=$System.OpacityScalar
    Scale RGB=&VectorFromFloat(Value=&FloatFromCurve(
      FloatCurve=DataInterface() {propertyValues {Curve {keys [
          {interpMode=RCIM_Cubic, leaveTangentWeight=0.4714, tangentMode=RCTM_User, time=0, value=1}
          {arriveTangentWeight=0.4714, interpMode=RCIM_Cubic, tangentMode=RCTM_User, time=1, value=0}
         ]}}}
      Scale Curve=5
     ))
    ScaleA=true
    ScaleRGB=true
   )
  ]
  rendererClasses [&NiagaraLightRendererProperties]
  renderers [{rendererClass=&NiagaraLightRendererProperties, rendererIndex=0}]
 }
 Squares {
  emitterUpdateScript [
   &EmitterState()
   &SpawnRate(
    Spawn Probability=&Lerp_Float(A=0, B=1, Alpha=$User.IsShieldIncreasing)
    SpawnRate=&FloatFromCurve(
     FloatCurve=DataInterface() {propertyValues {Curve {keys [
         {interpMode=RCIM_Cubic, tangentMode=RCTM_User, time=0, value=0}
         {interpMode=RCIM_Cubic, tangentMode=RCTM_User, time=0.05, value=1}
         {interpMode=RCIM_Cubic, tangentMode=RCTM_User, time=0.5, value=1}
         <at 0.6 0 RCIM_Cubic>
        ]}}}
     CurveIndex=$System.NormalizedAgeScaled
     Scale Curve=40
    )
    Use Spawn Probability=true
   )
  ]
  particleSpawnScript [
   &PartitionParticles(Number Of Partitions=17)
   &InitializeParticle(
    Color=<rgba 0.0052 0.3368 1 1>
    Lifetime Max=0.4
    Lifetime Min=0.2
    Lifetime Mode=::NewEnumerator1
    Position Offset=&Add_Vector(
     A=<0 0 10>
     B=&Multiply_VectorByFloat(
      Vector=<0 0 200>
      Float=&FloatFromCurve(
       FloatCurve=DataInterface() {propertyValues {Curve {keys [{tangentMode=RCTM_User, time=0, value=0}, {tangentMode=RCTM_User, time=0.8, value=1}, <at 1 0>]}}}
       CurveIndex=$System.NormalizedAgeScaled
      )
     )
    )
    Sprite Size Mode=::NewEnumerator1
    Uniform Sprite Size Max=2
    Uniform Sprite Size Min=1.5
    UsePositionOffset=true
    Write Material Random=true
   )
   &CylinderLocation(
    Cylinder Height=10
    Cylinder Radius=&Multiply_Float(A=&RandomRangeFloat(Minimum=15, Maximum=30), B=$System.GlowPanScale)
    MaskSpawnGroup=false
    Override Seed=false
    UseOwnerScale=true
   )
   &AddVelocity(Velocity=<0 0 1>, Velocity Speed Scale=&RandomRangeFloat(Minimum=50, Maximum=75)) {module=AddVelocity001}
   &AddVelocity(
    Velocity=$Particles.CylinderLocation.CylinderNormal
    Velocity Speed Scale=&RandomRangeFloat(Minimum=&RandomRangeFloat(Minimum=200, Maximum=200), Maximum=$System.GlowPanScale)
   ) {enabled=false, module=AddVelocity002}
  ]
  particleUpdateScript [
   &Drag(Drag=&RandomRangeFloat(Minimum=2, Maximum=3), Use Linear Drag=true, Use Rotational Drag=false)
   &SolveForcesAndVelocity(Clamp Velocity=false, Limit Acceleration=false)
   &ParticleState()
   &ScaleColor(
    Scale Alpha=&FloatFromCurve(
     FloatCurve=DataInterface() {propertyValues {Curve {keys [
         {arriveTangent=4.9067, interpMode=RCIM_Cubic, leaveTangent=4.9067, tangentMode=RCTM_User, time=0, value=0}
         {arriveTangentWeight=0.181, interpMode=RCIM_Cubic, leaveTangentWeight=0.7242, tangentMode=RCTM_User, time=0.15, value=1}
         <at 0.6 1>
         {arriveTangent=-0.0405, interpMode=RCIM_Cubic, leaveTangent=-0.0405, tangentMode=RCTM_User, time=1, value=0}
        ]}}}
     Scale Curve=$System.OpacityScalar
    )
    ScaleA=true
    ScaleRGB=true
   )
   &ScaleSpriteSize(
    Uniform Curve Scale=&RandomRangeFloat(Minimum=2, Maximum=$System.ShieldperSecondScale)
    Uniform Curve Sprite Scale=DataInterface() {propertyValues {Curve {keys [
        {arriveTangent=22.7901, interpMode=RCIM_Cubic, leaveTangent=22.7901, tangentMode=RCTM_User, time=0, value=0}
        {interpMode=RCIM_Cubic, tangentMode=RCTM_User, time=0.1, value=0.9847}
        {arriveTangent=-1.2347, interpMode=RCIM_Cubic, leaveTangent=-1.2347, tangentMode=RCTM_User, time=1, value=0}
       ]}}}
   )
  ]
  rendererClasses [&NiagaraSpriteRendererProperties]
  renderers [{rendererClass=&NiagaraSpriteRendererProperties}]
 }
 Wash_Decals001 {
  emitterUpdateScript [
   &EmitterState()
   &SpawnBurst_Instantaneous(Spawn Count=17, Use Spawn Probability=false)
  ]
  particleSpawnScript [
   &PartitionParticles(Number Of Partitions=17)
   &InitializeParticle(
    Color=$System.ColorB
    Lifetime=0.5
    Sprite Size Mode=::NewEnumerator3
    Uniform Sprite Size=75
    UsePositionOffset=false
    Write Material Random=true
   ) {module=InitializeParticle001}
   &InitializeParticle(Lifetime=0.5, Sprite Size Mode=::NewEnumerator3, Uniform Sprite Size=50, UsePositionOffset=false, Write Material Random=true) {enabled=false}
   &CameraOffset(Camera Offset Amount=25) {enabled=false}
   &CameraOffset(Camera Offset Amount=25) {module=CameraOffset001}
  ]
  particleUpdateScript [
   &ParticleState()
   &SkeletalMeshLocation(
    Bone / Socket Index=$Particles.PartitionParticles.Partition
    Bone Sampling Mode=::NewEnumerator4
    MaskSpawnGroup=false
    Skeletal Mesh=DataInterface() {propertyValues {
      DefaultMesh=null
      ExcludeBoneName=root
      OptionalComponentClass=null
      PreviewMesh=&M_Med_Soldier_04_Preview
      SkinningMode=SkinOnTheFly
      SoftSourceActor=null
      SourceMode=Default
      UvSetIndex=0
      WholeMeshLOD=-1
      bExcludeBone=false
      bReadDeformedGeometry=true
      bRequireCurrentFrameData=true
      ComponentTags []
      FilteredBones [
       hand_l
       hand_r
       ball_l
       ball_r
       lowerarm_l
       lowerarm_r
       upperarm_l
       upperarm_r
       thigh_l
       thigh_r
       calf_l
       calf_r
       calf_twist_02_l
       calf_twist_02_r
       thigh_twist_01_l
       thigh_twist_01_r
       spine_05
      ]
      FilteredSockets []
      MeshUserParameter {parameter {name=None, typeDefHandle {registeredTypeIndex=1}, varData []}}
      SamplingRegions []
     }}
   ) {enabled=false}
   &SkeletalMeshLocation(
    Bone / Socket Index=$Particles.PartitionParticles.Partition
    Bone Sampling Mode=::NewEnumerator4
    MaskSpawnGroup=false
    Sampled Position Offset=&SelectVectorFromArray(
     Direct Array Index=$Particles.PartitionParticles.Partition
     Vector Selection Array=DataInterface() {propertyValues {FloatData [
        <0 0 0>
        <5 0 0>
        <0 0 0>
        <0 0 0>
        <5 0 0>
        <0 0 0>
        <0 0 0>
        <-5 0 0>
        <-10 0 0>
        <8 0 0>
        <-8 0 0>
        <-4 0 0>
        <8 10 0>
        <8 0 0>
        <4 0 0>
        <-8 -10 0>
        <-10 -22 0>
       ]}}
    )
    Skeletal Mesh=$System.Skeletal Mesh
   ) {module=SkeletalMeshLocation001}
   &ScaleColor(
    Scale Alpha=&FloatFromCurve(FloatCurve=DataInterface() {propertyValues {Curve {keys [
         {arriveTangent=7.6727, interpMode=RCIM_Cubic, leaveTangent=7.6727, tangentMode=RCTM_User, time=0, value=0}
         {arriveTangent=-2.3136, interpMode=RCIM_Cubic, leaveTangent=-2.3136, tangentMode=RCTM_User, time=0.3, value=0.95}
         <at 1 0 RCIM_Cubic>
        ]}}})
    ScaleA=true
    ScaleRGB=true
   ) {enabled=false}
   &ScaleColor(
    Scale Alpha=&FloatFromCurve(FloatCurve=DataInterface() {propertyValues {Curve {keys [
         {interpMode=RCIM_Cubic, tangentMode=RCTM_User, time=0, value=0}
         {arriveTangent=-4.5912, interpMode=RCIM_Cubic, leaveTangent=-4.5912, tangentMode=RCTM_User, time=0.1, value=1.1}
         {tangentMode=RCTM_User, time=1, value=0}
        ]}}})
    Scale RGB=&VectorFromFloat(Value=&FloatFromCurve(
      FloatCurve=DataInterface() {propertyValues {Curve {keys [
          {arriveTangent=9.0725, interpMode=RCIM_Cubic, leaveTangent=9.0725, tangentMode=RCTM_User, time=0, value=0}
          {arriveTangent=-2.5249, interpMode=RCIM_Cubic, leaveTangent=-2.5249, tangentMode=RCTM_User, time=0.1408, value=0.9847}
          {arriveTangent=-0.0405, interpMode=RCIM_Cubic, leaveTangent=-0.0405, tangentMode=RCTM_User, time=1, value=0}
         ]}}}
      Scale Curve=2
     ))
    ScaleA=true
    ScaleRGB=true
   ) {enabled=false, module=ScaleColor002}
   &DynamicMaterialParameters(
    Index 0 Param 1=&FloatFromCurve(FloatCurve=DataInterface() {propertyValues {Curve {keys [
         {interpMode=RCIM_Cubic, leaveTangentWeight=0.0435, tangentMode=RCTM_User, time=0, value=-0.2}
         {arriveTangent=18.2867, interpMode=RCIM_Cubic, leaveTangent=18.2867, tangentMode=RCTM_User, time=0.07, value=0.275}
         {arriveTangent=5.0806, interpMode=RCIM_Cubic, leaveTangent=5.0806, time=0.15, value=1}
         {arriveTangent=5.6764, interpMode=RCIM_Cubic, leaveTangent=5.6764, tangentMode=RCTM_User, time=1, value=5}
        ]}}})
    Index 0 Param 2=1
    Index 0 Param 3=&FloatFromCurve(FloatCurve=DataInterface() {propertyValues {Curve {keys [{arriveTangent=2.2883, interpMode=RCIM_Cubic, leaveTangent=2.2883, tangentMode=RCTM_User, time=0, value=-0.1}, <at 1 1.8 RCIM_Cubic>]}}})
    Module.Param0WriteEnabled=true
    Module.Param1WriteEnabled=true
    Module.Param2WriteEnabled=true
    Module.Param3WriteEnabled=false
   ) {enabled=false}
   &DynamicMaterialParameters(
    Index 0 Param 1=&FloatFromCurve(
     FloatCurve=DataInterface() {propertyValues {Curve {keys [{interpMode=RCIM_Cubic, tangentMode=RCTM_User, time=0.1, value=0.5}, {tangentMode=RCTM_User, time=0.6, value=1}]}}}
     Scale Curve=&FloatFromCurve(
      FloatCurve=DataInterface() {propertyValues {Curve {keys [
          {interpMode=RCIM_Cubic, tangentMode=RCTM_User, time=0, value=1}
          {interpMode=RCIM_Cubic, tangentMode=RCTM_User, time=0.33, value=0.25}
          <at 0.66 1>
         ]}}}
      Scale Curve=4
     )
    )
    Index 0 Param 3=&FloatFromCurve(FloatCurve=DataInterface() {propertyValues {Curve {keys [<at 0 -0.1 RCIM_Cubic>, {interpMode=RCIM_Cubic, leaveTangent=1.6586, time=1, value=0.5}]}}})
    Module.Param0WriteEnabled=true
    Module.Param1WriteEnabled=false
    Module.Param2WriteEnabled=true
    Module.Param3WriteEnabled=false
   ) {module=DynamicMaterialParameters001}
   /SetVariables(Particles.Color=&MakeLinearColorFromVectorAndFloat(
     "Vector (RGB)"=&Multiply_VectorByFloat(
      Vector=&Lerp_Vector(
       A=&MakeVectorFromLinearColorRGB(Color=$System.ColorB)
       B=&MakeVectorFromLinearColorRGB(Color=<rgba 0.026 0.377 1 1>)
       Alpha=$User.IsShieldIncreasing
      )
      Float=$System.ShieldperSecondScale
     )
     "Float (Alpha)"=&FloatFromCurve(
      FloatCurve=DataInterface() {propertyValues {Curve {keys [
          {interpMode=RCIM_Cubic, tangentMode=RCTM_User, time=0, value=0.1}
          {interpMode=RCIM_Cubic, tangentMode=RCTM_User, time=0.5, value=1}
          <at 1 0.1>
         ]}}}
      Scale Curve=$System.OpacityScalar
     )
    )) {isSetParameters=true}
  ]
  rendererClasses [&NiagaraSpriteRendererProperties]
  renderers [
   {rendererClass=&NiagaraSpriteRendererProperties}
   {rendererClass=&NiagaraSpriteRendererProperties, rendererIndex=1}
  ]
 }
 Wisps {
  emitterSpawnScript [/SetVariables() {isSetParameters=true}]
  emitterUpdateScript [
   &EmitterState()
   &SpawnRate(SpawnRate=5, Use Spawn Probability=true) {enabled=false}
   &SpawnBurst_Instantaneous(Spawn Count=5, Use Spawn Probability=false, UseLoopCountLimit=false)
  ]
  particleSpawnScript [
   &InitializeParticle(
    Color Maximum=<rgba 0.9167 0.8 1 1>
    Color Minimum=<rgba 0.8667 0.8 1 1>
    Color Mode=::NewEnumerator2
    Lifetime Max=0.7
    Lifetime Min=0.5
    Lifetime Mode=::NewEnumerator1
    Mesh Scale Max=<0.4 0.4 0.8>
    Mesh Scale Min=<0.2 0.2 0.7>
    Mesh Scale Mode=::NewEnumerator2
    UsePositionOffset=false
    Write Material Random=true
   )
   &CameraOffset(Camera Offset Amount=5)
   &PartitionParticles(Number Of Partitions=5) {enabled=false}
   &InitialMeshOrientation(
    Mesh Axis To Orient=<0 0 1>
    Mesh Orientation Mode=::NewEnumerator2
    Orientation Vector=<0 0 1>
    Rotation=&RandomRangeVector(Minimum=<0 0 -1>, Maximum=<0 0 1>)
    Use Rotation Vector=true
   ) {module=InitialMeshOrientation001}
   &InitialMeshOrientation(
    Mesh Axis To Orient=<0 0 1>
    Mesh Orientation Mode=::NewEnumerator2
    Orientation Vector=<0 0 1>
    Rotation=&Multiply_VectorByFloat(Vector=<1 1 1>, Float=&Multiply_FloatByInt(Float=1, Integer=$Particles.PartitionParticles.Partition))
    Use Rotation Vector=true
   ) {enabled=false}
   &Color(Color=<rgba 1 1 1 0.9>)
  ]
  particleUpdateScript [
   &ParticleState()
   &ScaleMeshSize(Scale Factor=&MakeVector(
     X=&FloatFromCurve(FloatCurve=DataInterface() {propertyValues {Curve {keys [
          {arriveTangent=-0.0644, interpMode=RCIM_Cubic, leaveTangent=-0.0644, tangentMode=RCTM_User, time=0.3, value=1.2}
          {arriveTangent=-1.1459, interpMode=RCIM_Cubic, leaveTangent=-1.1459, tangentMode=RCTM_User, time=0.8, value=1}
         ]}}})
     Y=&FloatFromCurve(FloatCurve=DataInterface() {propertyValues {Curve {keys [
          {arriveTangent=-0.0644, interpMode=RCIM_Cubic, leaveTangent=-0.0644, tangentMode=RCTM_User, time=0.3, value=1.2}
          {arriveTangent=-1.1459, interpMode=RCIM_Cubic, leaveTangent=-1.1459, tangentMode=RCTM_User, time=0.8, value=1}
         ]}}})
     Z=1
    )) {enabled=false}
   &ScaleColor(
    Scale Alpha=&FloatFromCurve(
     FloatCurve=DataInterface() {propertyValues {Curve {keys [
         {interpMode=RCIM_Cubic, tangentMode=RCTM_User, time=0, value=0}
         {interpMode=RCIM_Cubic, tangentMode=RCTM_User, time=0.2, value=1}
         {interpMode=RCIM_Cubic, tangentMode=RCTM_User, time=0.9, value=0}
        ]}}}
     Scale Curve=$System.OpacityScalar
    )
    ScaleA=true
    ScaleRGB=true
   )
   &DynamicMaterialParameters(
    Index 0 Param 1=&FloatFromCurve(FloatCurve=DataInterface() {propertyValues {Curve {keys [
         {arriveTangent=-3.115, interpMode=RCIM_Cubic, leaveTangent=-3.115, tangentMode=RCTM_User, time=0, value=0.7}
         {arriveTangent=-0.0788, interpMode=RCIM_Cubic, leaveTangent=-0.0788, tangentMode=RCTM_User, time=0.35, value=0.05}
         {arriveTangent=-1.1166, interpMode=RCIM_Cubic, leaveTangent=-1.1166, tangentMode=RCTM_User, time=0.6, value=-0.1}
         {arriveTangent=-3.8204, interpMode=RCIM_Cubic, leaveTangent=-3.8204, tangentMode=RCTM_User, time=1, value=-1}
        ]}}})
    Index 0 Param 2=&RandomRangeFloat(Minimum=0.9, Maximum=1.4)
    Module.Index0Param0Write=true
    Module.Index0Param1Write=true
    Module.Index0Param2Write=false
    Module.Index0Param3Write=false
   ) {enabled=false}
   &DynamicMaterialParameters(
    Index 0 Param 1=&FloatFromCurve(FloatCurve=DataInterface() {propertyValues {Curve {keys [
         {arriveTangent=-2.0701, interpMode=RCIM_Cubic, leaveTangent=-2.0701, tangentMode=RCTM_User, time=0, value=0.1}
         {arriveTangent=-0.2198, interpMode=RCIM_Cubic, leaveTangent=-0.2198, tangentMode=RCTM_User, time=0.7, value=-1}
        ]}}})
    Index 0 Param 2=&RandomRangeFloat(Minimum=0.9, Maximum=1.5)
    Index 0 Param 3=&FloatFromCurve(FloatCurve=DataInterface() {propertyValues {Curve {keys [
         {interpMode=RCIM_Cubic, tangentMode=RCTM_User, time=0, value=1}
         <at 0.2 0>
         {arriveTangent=1.25, interpMode=RCIM_Cubic, leaveTangent=0.1873, tangentMode=RCTM_Break, time=0.6943, value=0}
         {arriveTangent=0.2193, interpMode=RCIM_Cubic, leaveTangent=0.2193, tangentMode=RCTM_User, time=1, value=1.1}
        ]}}})
    Module.Index0Param0Write=true
    Module.Index0Param1Write=true
    Module.Index0Param2Write=true
    Module.Index0Param3Write=false
   ) {module=DynamicMaterialParameters001}
   &DynamicMaterialParameters(
    Index 0 Param 1=-1
    Index 0 Param 2=1
    Index 0 Param 3=1
    Module.Index0Param0Write=true
    Module.Index0Param1Write=true
    Module.Index0Param2Write=true
    Module.Index0Param3Write=false
   ) {enabled=false, module=DynamicMaterialParameters002}
  ]
  rendererClasses [&NiagaraMeshRendererProperties]
  renderers [{rendererClass=&NiagaraMeshRendererProperties, rendererIndex=0}]
 }
}
summary {
 systemName=NS_SunRose_Medallion_SiphonHealingV2
 emitters [
  {emitterName=EdgeGlow, rendererClasses [&NiagaraMeshRendererProperties]}
  {emitterName=Wisps, rendererClasses [&NiagaraMeshRendererProperties]}
  {emitterName=Flares, rendererClasses [&NiagaraSpriteRendererProperties]}
  {emitterName=Light, rendererClasses [&NiagaraLightRendererProperties]}
  {emitterName=CrossSparkles, rendererClasses [&NiagaraSpriteRendererProperties]}
  {emitterName=Squares, rendererClasses [&NiagaraSpriteRendererProperties]}
  {emitterName=Wash_Decals001, rendererClasses [&NiagaraSpriteRendererProperties]}
 ]
 userVariables [
  User.IsShieldIncreasing {type=NiagaraFloat, defaultValue=1, description=""}
  User.Shields {type=NiagaraFloat, defaultValue=0, description=""}
 ]
}
